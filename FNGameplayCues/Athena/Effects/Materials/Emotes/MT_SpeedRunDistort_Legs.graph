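# Translucent unlit particle material for the speed-run emote's leg effect, judging by its name, combining scrolling smoke with refractive distortion.
# Tiled smoke and normal textures pan along V, with the normal's green channel breaking up a contrast-adjusted 2D sphere mask for depth-faded opacity.
# The mask also blends the normal from flat and, scaled by a dynamic parameter, drives depth-faded index-of-refraction distortion.
# Particle color supplies the emissive through a shared adjustment function and scales opacity.
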Material /FNGameplayCues/Athena/Effects/Materials/Emotes/MT_SpeedRunDistort_Legs
graph {
 properties {
  bUsedWithNiagaraSprites=true
  bUsedWithParticleSprites=true
  blendMode=BLEND_Translucent
  pixelDepthOffsetMode=PDOM_Legacy
  refractionMethod=RM_IndexOfRefraction
  shadingModel=MSM_Unlit
  usedShadingModels=Unlit
 }
 nodes [
  n1 = ParticleColor()
  n2 = VectorParameter() {parameterName=Tile, defaultValue=<rgba 1 1 1 1>}
  n3 = Panner(
   Coordinate=Multiply(A=TextureCoordinate(), B=ComponentMask(None=@n2.RGB) {g=true, r=true})
   Speed=AppendVector(A=Constant(), B=ScalarParameter() {defaultValue=3, parameterName=PanSpeed})
  )
  n4 = TextureSample(UVs=@n3) {texture=&T_Inky_Smoke_Tile}
  n5 = MaterialFunctionCall!(Hardness=ScalarParameter() {parameterName=EdgeHardness}, Center, Radius) {materialFunction=/Game/Packages/Fortress_SharedMaterials/Base_Material_Functions/FX_SphereMask2D}
  n6 = TextureSample(UVs=@n3) {samplerType=SAMPLERTYPE_Normal, texture=/Game/Effects/Fort_Effects/Gameplay/SmokeDustExplosions/Textures/T_Explosion_NM_01}
  n7 = ConstantBiasScale(None=@n6.G)
  n8 = MaterialFunctionCall(In=Multiply(A=Add(A=@n5.Result, B=@n7), B=@n5.Result), Contrast=ScalarParameter() {defaultValue=2, parameterName=Contrast}) {materialFunction=&CheapContrast}
  DepthFade(Opacity=Multiply(A=@n1.A, B=Multiply(A=@n4.G, B=@n8.Result)), FadeDistance=ScalarParameter() {defaultValue=16, parameterName=DFDist})
  LinearInterpolate(A=Constant3Vector() {constant=<rgba 0 0 1 0>}, B=@n6.RGB, Alpha=@n5.Result)
  n9 = DynamicParameter() {paramNames [RefractionAMnt, Param2, Param3, Param4]}
  LinearInterpolate(
   B=OneMinus(None=ScalarParameter() {defaultValue=0.15, parameterName=RefractionAmnt})
   Alpha=DepthFade(Opacity=Multiply(A=@n5.Result, B=@n9.RefractionAMnt), FadeDistance=ScalarParameter() {defaultValue=16, parameterName=DFDist})
  ) {constA=1}
  MaterialFunctionCall(Emissive=@n1.RGB) {materialFunction=&MF_FX_EmissiveAdjustments}
 ]
}
